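# Additive unlit material for Niagara mesh particles that draws a glowing, tinted eye for the Logarithm glider's FX.
# Bump-offset UV1 adds parallax to a vertical gradient and a centered sphere mask, with radius driven per particle by the EyeRadius dynamic parameter.
# Eye color and particle color tint the result, which passes through HDR derivation with per-particle intensity and emissive adjustments.
# A powered UV0 gradient scales a local-space directional vector for world position offset.
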Material /Game/Gadgets/Assets/VinderTech_GliderChute/Glider_Logarithm/FX/M_Glider_Logarithm_Eye
graph {
 properties {
  bUsedWithNiagaraMeshParticles=true
  blendMode=BLEND_Additive
  pixelDepthOffsetMode=PDOM_Legacy
  previewMesh=/Game/Effects/Fort_Effects/Mesh/SM_HalfSphere_900polys
  shadingModel=MSM_Unlit
  usedShadingModels=Unlit
 }
 nodes [
  n1 = BumpOffset(Coordinate=TextureCoordinate() {coordinateIndex=1}, Height=ScalarParameter() {defaultValue=0.5, parameterName=Height})
  n2 = ComponentMask(None=@n1) {g=true}
  n3 = VectorParameter() {parameterName=EyeColor, defaultValue=<rgba 1 0.01 0.01 0>}
  n4 = DynamicParameter() {paramNames [EyeRadius, HDR Intensity, Param3, Param4]}
  n5 = SphereMask(
   A=@n1
   B=Constant() {r=0.5}
   Radius=Clamp(None=LinearInterpolate(Alpha=@n4.EyeRadius) {constB=0.4}) {maxDefault=0.4}
   Hardness=ScalarParameter() {parameterName=Hardness}
  )
  n6 = ParticleColor()
  n7 = Multiply(A=Multiply(A=Multiply(A=@n2, B=@n3.RGB), B=@n5), B=@n6.RGB)
  MaterialFunctionCall(Emissive=MaterialFunctionCall(
    LDR Input=Clamp(None=@n7)
    HDR Tint=@n7
    HDR Intensity Multiplier=DynamicParameter() {paramNames [EyeRadius, HDR Intensity, Param3, Param4]}.HDR Intensity
    Derive HDR Power=ScalarParameter() {defaultValue=2, parameterName=HDR Power Derive}
    Derive Luminance from Linear=StaticBool() {value=true}
   ) {materialFunction=&DeriveHDRfromLDR}.RemappedRange) {materialFunction=&MF_FX_EmissiveAdjustments}
  n8 = ComponentMask(None=TextureCoordinate()) {g=true}
  n9 = ScalarParameter() {defaultValue=1.5, parameterName=WPOMaskContrast}
  n10 = VectorParameter() {parameterName=WPODirectionalVector, defaultValue=<rgba -1 0 0 0>}
  Multiply(
   A=Saturate(None=Power(Base=@n8, Exp=@n9))
   B=Multiply(A=Transform(None=@n10.RGB) {transformSourceType=TRANSFORMSOURCE_Local}, B=ScalarParameter() {defaultValue=30, parameterName=WPOAmount})
  )
  Multiply(A=@n6.A, B=@n5)
 ]
}
