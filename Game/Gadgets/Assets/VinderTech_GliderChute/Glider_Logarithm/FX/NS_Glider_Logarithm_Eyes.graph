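# Niagara system for the Logarithm glider that adds glowing red, blinking eyes and flight speedlines.
# Eyes and Central_Eye pin mesh particles with lights to indexed bones or sockets on the glider's skeletal mesh.
# System timelines feed a Blink_Control_Setup input, split into six groups, driving material, size and light radius, plus distance-remapped brightness flashes.
# GPU Speedlines spawn by distance moved when fully deployed and ForwardDot passes a threshold, stretching with speed.
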NiagaraSystem /Game/Gadgets/Assets/VinderTech_GliderChute/Glider_Logarithm/FX/NS_Glider_Logarithm_Eyes
dependencies {
 usedDataInterfaces [
  &NiagaraDataInterfaceArrayFloat
  &NiagaraDataInterfaceCurve
  &NiagaraDataInterfaceVector2DCurve
  &NiagaraDataInterfaceVectorCurve
  &NiagaraDataInterfaceArrayFloat2
  &NiagaraDataInterfaceArrayFloat3
  &NiagaraDataInterfaceArrayColor
  &NiagaraDataInterfaceColorCurve
  &NiagaraDataInterfaceArrayInt32
 ]
 usedDynamicInputs [&Waveform]
 usedModules [
  &SystemState
  &self:SystemSpawnScript.SystemScriptSource.SystemScriptGraph.NiagaraNodeAssignment_2.SetVariables_#1
  &Timeline
 ]
}
emitters {
 Central_Eye {
  emitterUpdateScript [
   &EmitterState(Life Cycle Mode=::NewEnumerator1, Loop Behavior=::NewEnumerator1, Loop Duration Mode=::NewEnumerator1, UseLoopDelay=false)
   &SpawnBurst_Instantaneous(Use Spawn Probability=false, UseLoopCountLimit=false)
  ]
  particleSpawnScript [
   &InitializeParticle(
    Color=<rgba 1 0.0747 0.0747 1>
    Mesh Scale=<0.05 0.05 0.035>
    Mesh Scale Mode=::NewEnumerator4
    UsePositionOffset=false
    Write Material Random=true
   )
   /SetVariables(Particles.SpawnGroup=31) {isSetParameters=true}
   &CameraOffset(Camera Offset Amount=1)
  ]
  particleUpdateScript [
   &ParticleState(Kill Particles When Lifetime Has Elapsed=false, Let Infinitely Lived Particles Die When Emitter Deactivates=true)
   &SkeletalMeshLocation(
    Bone / Socket Index=&SpawnGroup()
    MaskSpawnGroup=false
    Mesh Sampling Type=::NewEnumerator5
    Skeletal Mesh=$User.Glider
    Socket Sampling Mode=::NewEnumerator5
   )
   &DynamicMaterialParameters(
    Index 0 Param 1=&Multiply_Float(
     A=1
     B=&self:Blink_Control_Setup(
      Group A:NiagaraFloat=$System.Timeline.Float
      Group B:NiagaraFloat=$System.Timeline001.Float
      Group C:NiagaraFloat=$System.Timeline002.Float
      Group D:NiagaraFloat=$System.Timeline003.Float
      Group E:NiagaraFloat=$System.Timeline004.Float
      Group F:NiagaraFloat=$System.Timeline005.Float
     )
    )
    Index 0 Param 2=3
    Index 1 Param 1=1
    Index 1 Param 2=1
    Index 1 Param 3=1
    Index 1 Param 4=1
    Module.Param0WriteEnabled=true
    Module.Param0WriteEnabled1=true
    Module.Param1WriteEnabled=true
    Module.Param1WriteEnabled1=true
    Module.Param2WriteEnabled=true
    Module.Param2WriteEnabled1=true
    Module.Param3WriteEnabled=true
    Module.Param3WriteEnabled1=true
    Write Parameter Index 1=true
   )
   &ScaleSpriteSize(
    Scale Sprite Size Mode=::NewEnumerator0
    Uniform Scale Factor=&Multiply_Float(
     A=1
     B=&self:Blink_Control_Setup(
      Group A:NiagaraFloat=$System.Timeline.Float
      Group B:NiagaraFloat=$System.Timeline001.Float
      Group C:NiagaraFloat=$System.Timeline002.Float
      Group D:NiagaraFloat=$System.Timeline003.Float
      Group E:NiagaraFloat=$System.Timeline004.Float
      Group F:NiagaraFloat=$System.Timeline005.Float
     )
    )
   )
   /SetVariables(
    Particles.LightColor=<rgba 7 0.5226 0.5226 1>
    Particles.LightExponent=2
    Particles.LightPosition:Vector3f=&Add_Vector(A=Linked() {linkedVariable {name=Particles.Position, type=NiagaraPosition}}, B=<0 -3 0>)
    Particles.LightRadius=&Multiply_Float(
     A=6
     B=&self:Blink_Control_Setup(
      Group A:NiagaraFloat=$System.Timeline.Float
      Group B:NiagaraFloat=$System.Timeline001.Float
      Group C:NiagaraFloat=$System.Timeline002.Float
      Group D:NiagaraFloat=$System.Timeline003.Float
      Group E:NiagaraFloat=$System.Timeline004.Float
      Group F:NiagaraFloat=$System.Timeline005.Float
     )
    )
   ) {isSetParameters=true}
   &ScaleColor(
    Scale RGB=&VectorFromFloat(Value=&ClampFloat(
      Float=&FN_LOD_Distance_Remap(
       Start Distance=500
       End Distance=1500
       Minimum Output=&FloatFromCurve(
        FloatCurve=DataInterface() {propertyValues {Curve {keys [
            {interpMode=RCIM_Cubic, leaveTangentWeight=0.1667, tangentMode=RCTM_User, time=0, value=1}
            {arriveTangentWeight=0.1667, interpMode=RCIM_Cubic, leaveTangentWeight=0.1667, tangentMode=RCTM_User, time=0.25, value=200}
            {arriveTangent=-882.7394, arriveTangentWeight=68.5719, interpMode=RCIM_Cubic, leaveTangent=-882.7394, tangentMode=RCTM_User, time=1, value=1}
           ]}}}
        CurveIndex=$System.Timeline006.NormalizedTime
       )
       Maximum Output=&FloatFromCurve(
        FloatCurve=DataInterface() {propertyValues {Curve {keys [
            {interpMode=RCIM_Cubic, leaveTangentWeight=184.3128, tangentMode=RCTM_User, time=0, value=1}
            {arriveTangentWeight=0.1667, interpMode=RCIM_Cubic, leaveTangentWeight=0.1667, tangentMode=RCTM_User, time=0.25, value=400}
            {arriveTangent=-1742.759, arriveTangentWeight=102.7615, interpMode=RCIM_Cubic, leaveTangent=-1742.759, tangentMode=RCTM_User, time=1, value=1}
           ]}}}
        CurveIndex=$System.Timeline006.NormalizedTime
       )
      )
      Max=5
     ))
    ScaleA=true
    ScaleRGB=true
   )
   &ScaleColor(
    Scale RGB=&VectorFromFloat(Value=&ClampFloat(
      Float=&FN_LOD_Distance_Remap(Start Distance=500, End Distance=1500, Minimum Output=1, Maximum Output=10)
      Min=1
      Max=10
     ))
    ScaleA=true
    ScaleRGB=true
   ) {enabled=false, module=ScaleColor001}
  ]
  rendererClasses [&NiagaraMeshRendererProperties, &NiagaraLightRendererProperties]
  renderers [
   {rendererClass=&NiagaraMeshRendererProperties, rendererIndex=0}
   {rendererClass=&NiagaraLightRendererProperties, rendererIndex=1}
  ]
 }
 Eyes {
  emitterUpdateScript [
   &EmitterState(Life Cycle Mode=::NewEnumerator1, Loop Behavior=::NewEnumerator1, Loop Duration Mode=::NewEnumerator1, UseLoopDelay=false)
   &SpawnBurst_Instantaneous(Spawn Count=31, Use Spawn Probability=false, UseLoopCountLimit=false)
  ]
  particleSpawnScript [
   &FN_SetSpawnGroup(Number of Spawn Groups=31)
   &InitializeParticle(
    Color=<rgba 1 0.0747 0.0747 1>
    Mesh Scale Mode=::NewEnumerator3
    Mesh Uniform Scale=&SelectFloatFromArray(
     Direct Array Index=$Particles.SpawnGroup
     Float Selection Array=DataInterface() {propertyValues {FloatData [0.045,0.03,0.0175,0.0125,0.01,0.045,0.03,0.0175,0.0125,0.01,0.0275,0.0225,0.0225,0.0275,0.0225,0.0225,0.015,0.0125,0.0185,0.015,0.0125,0.015,0.0125,0.0185,0.015,0.0125,0.0175,0.02,0.015,0.015,0.02]}}
    )
    UsePositionOffset=false
    Write Material Random=true
   )
   &CameraOffset(Camera Offset Amount=5)
  ]
  particleUpdateScript [
   &ParticleState(Kill Particles When Lifetime Has Elapsed=false, Let Infinitely Lived Particles Die When Emitter Deactivates=true)
   &SkeletalMeshLocation(
    Bone / Socket Index=$Particles.SpawnGroup
    MaskSpawnGroup=false
    Mesh Sampling Type=::NewEnumerator5
    Skeletal Mesh=$User.Glider
    Socket Sampling Mode=::NewEnumerator5
   )
   &DynamicMaterialParameters(
    Index 0 Param 1=&Multiply_Float(
     A=1
     B=&self:Blink_Control_Setup(
      Group A:NiagaraFloat=$System.Timeline.Float
      Group B:NiagaraFloat=$System.Timeline001.Float
      Group C:NiagaraFloat=$System.Timeline002.Float
      Group D:NiagaraFloat=$System.Timeline003.Float
      Group E:NiagaraFloat=$System.Timeline004.Float
      Group F:NiagaraFloat=$System.Timeline005.Float
     )
    )
    Index 0 Param 2=3
    Index 1 Param 1=1
    Index 1 Param 2=1
    Index 1 Param 3=1
    Index 1 Param 4=1
    Module.Param0WriteEnabled=true
    Module.Param0WriteEnabled1=true
    Module.Param1WriteEnabled=true
    Module.Param1WriteEnabled1=true
    Module.Param2WriteEnabled=true
    Module.Param2WriteEnabled1=true
    Module.Param3WriteEnabled=true
    Module.Param3WriteEnabled1=true
    Write Parameter Index 1=true
   )
   &ScaleSpriteSize(
    Scale Sprite Size Mode=::NewEnumerator0
    Uniform Scale Factor=&Multiply_Float(
     A=1
     B=&self:Blink_Control_Setup(
      Group A:NiagaraFloat=$System.Timeline.Float
      Group B:NiagaraFloat=$System.Timeline001.Float
      Group C:NiagaraFloat=$System.Timeline002.Float
      Group D:NiagaraFloat=$System.Timeline003.Float
      Group E:NiagaraFloat=$System.Timeline004.Float
      Group F:NiagaraFloat=$System.Timeline005.Float
     )
    )
   )
   /SetVariables(
    Particles.LightColor=<rgba 7 0.5226 0.5226 1>
    Particles.LightExponent:NiagaraFloat=&SelectFloatFromArray(
     Direct Array Index=$Particles.SpawnGroup
     Float Selection Array=DataInterface() {propertyValues {FloatData [2,5,2,2,3,2,5,2,2,3,2,2,2,2,2,2,2,2,2,2,2,2,2,2,2,2,2,2,2,2,2]}}
    )
    Particles.LightPosition:NiagaraPosition=$Particles.Position
    Particles.LightRadius=&Multiply_Float(
     A=&SelectFloatFromArray(
      Direct Array Index=$Particles.SpawnGroup
      Float Selection Array=DataInterface() {propertyValues {FloatData [8,6,3,2,3,8,6,3,2,3,5,5,5,5,5,5,5,5,5,5,5,5,5,5,5,5,5,5,5,5,5]}}
     )
     B=&self:Blink_Control_Setup(
      Group A:NiagaraFloat=$System.Timeline.Float
      Group B:NiagaraFloat=$System.Timeline001.Float
      Group C:NiagaraFloat=$System.Timeline002.Float
      Group D:NiagaraFloat=$System.Timeline003.Float
      Group E:NiagaraFloat=$System.Timeline004.Float
      Group F:NiagaraFloat=$System.Timeline005.Float
     )
    )
   ) {isSetParameters=true}
   &ScaleColor(
    Scale RGB=&VectorFromFloat(Value=&ClampFloat(
      Float=&FN_LOD_Distance_Remap(
       Start Distance=500
       End Distance=1500
       Minimum Output=&FloatFromCurve(
        FloatCurve=DataInterface() {propertyValues {Curve {keys [
            {interpMode=RCIM_Cubic, leaveTangentWeight=0.1667, tangentMode=RCTM_User, time=0, value=1}
            {arriveTangentWeight=0.1667, interpMode=RCIM_Cubic, leaveTangentWeight=0.1667, tangentMode=RCTM_User, time=0.25, value=200}
            {arriveTangent=-882.7394, arriveTangentWeight=68.5719, interpMode=RCIM_Cubic, leaveTangent=-882.7394, tangentMode=RCTM_User, time=1, value=1}
           ]}}}
        CurveIndex=&self:Blink_Control_Setup(
         Group A:NiagaraFloat=$System.Timeline006.NormalizedTime
         Group B:NiagaraFloat=$System.Timeline007.NormalizedTime
         Group C:NiagaraFloat=$System.Timeline008.NormalizedTime
         Group D:NiagaraFloat=$System.Timeline009.NormalizedTime
         Group E:NiagaraFloat=$System.Timeline010.NormalizedTime
         Group F:NiagaraFloat=$System.Timeline011.NormalizedTime
        )
       )
       Maximum Output=&FloatFromCurve(
        FloatCurve=DataInterface() {propertyValues {Curve {keys [
            {interpMode=RCIM_Cubic, leaveTangentWeight=184.3128, tangentMode=RCTM_User, time=0, value=1}
            {arriveTangentWeight=0.1667, interpMode=RCIM_Cubic, leaveTangentWeight=0.1667, tangentMode=RCTM_User, time=0.25, value=400}
            {arriveTangent=-1742.759, arriveTangentWeight=102.7615, interpMode=RCIM_Cubic, leaveTangent=-1742.759, tangentMode=RCTM_User, time=1, value=1}
           ]}}}
        CurveIndex=&self:Blink_Control_Setup(
         Group A:NiagaraFloat=$System.Timeline006.NormalizedTime
         Group B:NiagaraFloat=$System.Timeline007.NormalizedTime
         Group C:NiagaraFloat=$System.Timeline008.NormalizedTime
         Group D:NiagaraFloat=$System.Timeline009.NormalizedTime
         Group E:NiagaraFloat=$System.Timeline010.NormalizedTime
         Group F:NiagaraFloat=$System.Timeline011.NormalizedTime
        )
       )
      )
      Max=5
     ))
    ScaleA=true
    ScaleRGB=true
   )
   &ScaleColor(
    Scale RGB=&VectorFromFloat(Value=&ClampFloat(
      Float=&FN_LOD_Distance_Remap(Start Distance=500, End Distance=1500, Minimum Output=1, Maximum Output=10)
      Min=1
      Max=10
     ))
    ScaleA=true
    ScaleRGB=true
   ) {module=ScaleColor001}
  ]
  rendererClasses [&NiagaraMeshRendererProperties, &NiagaraLightRendererProperties]
  renderers [
   {rendererClass=&NiagaraMeshRendererProperties, rendererIndex=0}
   {rendererClass=&NiagaraLightRendererProperties, rendererIndex=1}
  ]
 }
 Speedlines {
  simTarget=GPUComputeSim
  emitterUpdateScript [
   &EmitterState()
   &SpawnRate(SpawnRate=100, Use Spawn Probability=false) {enabled=false}
   &SpawnPerUnit(
    Module.Use Max Movement Threshold=true
    Module.Use Movement Tolerance=true
    Movement Tolerance=10
    Spawn Spacing:NiagaraFloat=&MakeCustomFloatFromBool(
     Bool=&Boolean_AndOperation(
      Bool1=$User.bisFullyDeployed
      Bool2=&SetBoolByFloatComparison(A=$User.ForwardDot, B=0.1, Comparison Type=::NewEnumerator2)
     )
     True Float=40
     False Float=0
    )
    Use Spawn Probability=false
   )
  ]
  particleSpawnScript [
   &InitializeParticle(
    Color=&MakeLinearColorFromVectorAndFloat(
     "Vector (RGB)"=&MakeVectorFromLinearColorRGB(Color=<rgba 0.2292 0.1989 0.1838 1>)
     "Float (Alpha)"=&RandomRangeFloat(Minimum=0.125, Maximum=0.333)
    )
    Lifetime Max=0.4
    Lifetime Min=0.2
    Lifetime Mode=::NewEnumerator1
    Sprite Size Max=<5 4>
    Sprite Size Min=<3 2>
    Sprite Size Mode=::NewEnumerator2
    UsePositionOffset=false
    Write Material Random=true
   )
   &SkeletalMeshLocation(
    Bone Sampling Mode=::NewEnumerator4
    MaskSpawnGroup=false
    Orientation Sampling=::NewEnumerator1
    Position Sampling=::NewEnumerator2
    Sampled Position Offset=<0 -150 0>
    Skeletal Mesh=$User.Glider
   )
   &TorusLocation(
    Align Torus With Axis=true
    Coordinate Space:ENiagaraCoordinateSpace=::Simulation
    Handle Radius=75
    Large Radius=200
    MaskSpawnGroup=false
    Module.Override Randomness=false
    Torus Axis=&TransformVector(Vector=<0 1 0>, Destination Space=::Simulation)
    Torus Origin:NiagaraPosition=$Particles.SkeletalMeshLocation.SampledPosition
   )
   &InheritVelocity(Module.Limit Speed=true, Velocity Scale=<0.7 0.7 0>)
  ]
  particleUpdateScript [
   &ParticleState()
   &SolveForcesAndVelocity(Clamp Velocity=false, Limit Acceleration=false)
   &ScaleSpriteSize(Scale Factor=&Vector2DFromCurve(
     Vector2Curve=DataInterface() {propertyValues {
       XCurve {keys [<at -0.0069 0.5 RCIM_Cubic>,<at 0.4918 1 RCIM_Cubic>,<at 1 0.5 RCIM_Cubic>]}
       YCurve {keys [<at 0 0.0245 RCIM_Cubic>, {arriveTangent=-0.0245, interpMode=RCIM_Cubic, leaveTangent=-0.0245, time=0.4975, value=1}, <at 1 0 RCIM_Cubic>]}
      }}
     ScaleCurve=<1 2>
    ))
   &ScaleSpriteSizeBySpeed(Max Scale Factor=<1 100>, Min Scale Factor=<1 10>)
  ]
  rendererClasses [&NiagaraSpriteRendererProperties]
  renderers [{rendererClass=&NiagaraSpriteRendererProperties}]
 }
}
summary {
 systemName=NS_Glider_Logarithm_Eyes
 emitters [
  {emitterName=Eyes, rendererClasses [&NiagaraMeshRendererProperties, &NiagaraLightRendererProperties]}
  {emitterName=Central_Eye, rendererClasses [&NiagaraMeshRendererProperties, &NiagaraLightRendererProperties]}
  {emitterName=Speedlines, simTarget=GPUComputeSim, rendererClasses [&NiagaraSpriteRendererProperties]}
 ]
 userVariables [
  User.bisFrontEnd {type=NiagaraBool}
  User.bisFullyDeployed {type=NiagaraBool, defaultValue=true}
  User.ForwardDot {type=NiagaraFloat, defaultValue=0, description=""}
  User.Glider {
   type=NiagaraDataInterfaceSkeletalMesh
   defaultValue {
    struct=NiagaraExt_VariableValue_DataInterface
    value {dataInterface=&self:NiagaraDataInterfaceSkeletalMesh_0}
   }
  }
  User.ManualBlinkToggle {type=NiagaraBool, defaultValue=true}
 ]
}
